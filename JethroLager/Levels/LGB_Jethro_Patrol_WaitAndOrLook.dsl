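; Placeable local gameplay behavior for henchman AI patrols that makes the activating bot wait, optionally facing a set direction.
; On execute, it turns the bot controller to the direction arrow, swaps biome-tagged effects for an optional location GameplayEffect, then finishes after a random min-max delay.
; The construction script keeps the actor yaw-only and destroys the arrow when LookWhileWaiting is off, so the arrow's validity check skips the turn.
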(event LocalGameplayBehavior|EventOnExecute
  (params (Activator :type Actor%20Object%20Reference))
  (bind _returnvalue (Ability|GetFortAbilitySystemComponent Activator))
  (Utilities|FlowControl|Sequence
    (:then_0
      (Utilities|IsValid (Variables|Default|GetDirectionArrow)
        (:Is%20Valid
          (bind _asfort_athena_aibot_controller (Utilities|Casting|CastToFortAthenaAIBotController Activator))
          (Pawn|SetControlRotation _asfort_athena_aibot_controller (Transformation|GetWorldRotation (Variables|Default|GetDirectionArrow))))))
    (:then_1
      (if (and (Utilities|IsValid _returnvalue :!class /Script/Engine.KismetSystemLibrary) (Utilities|IsValidClass (Variables|Default|GetLocationGE)))
        (GameplayEffects|RemoveActiveEffectswithGrantedTags _returnvalue "(GameplayTags=((TagName=\"Athena.Location.Biome.Coastal\"),(TagName=\"Athena.Location.Biome.Forest\"),(TagName=\"Athena.Location.Biome.Grassland\"),(TagName=\"Athena.Location.Biome.Mountain\")))")
        (GameplayEffects|ApplyGameplayEffectToSelf :self _returnvalue :GameplayEffectClass (Variables|Default|GetLocationGE) :Level 0.0)))
    (:then_2
      (Utilities|FlowControl|Delay (Math|Random|RandomFloatinRange (Variables|HenchmanSetup|GetWaitTimeMin) (Variables|HenchmanSetup|GetWaitTimeMax)))
      (LocalGameplayBehavior|FinishExecute))))

(fn ConstructionScript
  (Transformation|SetActorRotation (Math|Rotator|MakeRotator 0.0 0.0 (.yaw (Transformation|GetActorRotation)))) ; Make sure we don't pitch or roll the actor. This whole thing is to set YAW of the AI
  (if (not (Variables|HenchmanSetup|GetLookWhileWaiting))
    (Components|DestroyComponent (Variables|Default|GetDirectionArrow))))
